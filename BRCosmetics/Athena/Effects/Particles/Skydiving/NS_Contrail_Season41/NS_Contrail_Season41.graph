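# Season 41 skydiving contrail cosmetic: glowing orange-gold meshes, cube trails, lightning and rock shards attached to the player's skeletal mesh.
# Emitters partition particles across sampled bones or sockets; Glows adds sprites and lights.
# Base packs four mesh layers into one emitter, using spawn groups to switch inputs and set renderer visibility tags.
# FN_Contrail_Setup velocity pulls cubes into trails, while glide lean, time of day and front-end state modulate the look.
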NiagaraSystem /BRCosmetics/Athena/Effects/Particles/Skydiving/NS_Contrail_Season41/NS_Contrail_Season41
dependencies {
 usedDataInterfaces [
  /Script/Niagara.NiagaraDataInterfaceArrayDistributionInt
  &NiagaraDataInterfaceCurve
  &NiagaraDataInterfaceArrayFloat3
  &NiagaraDataInterfaceMeshRendererInfo
  &NiagaraDataInterfaceCamera
  &NiagaraDataInterfaceVectorCurve
  &NiagaraDataInterfaceArrayInt32
  &NiagaraDataInterfacePlatformSet
  &NiagaraDataInterfaceSkeletalMesh
  &NiagaraDataInterfaceColorCurve
 ]
 usedDynamicInputs []
 usedModules [
  &self:SystemSpawnScript.SystemScriptSource.SystemScriptGraph.NiagaraNodeAssignment_0.SetVariables_#1
  &self:SystemSpawnScript.SystemScriptSource.SystemScriptGraph.NiagaraNodeAssignment_1.SetVariables_#2
  &SystemState
  /Game/Effects/Niagara/Modules/Contrails/FN_Contrail_Setup
  &FN_Integrator
 ]
}
emitters {
 Base {
  emitterUpdateScript [
   &EmitterState(Life Cycle Mode=::NewEnumerator1, Loop Behavior=::NewEnumerator1, Loop Duration Mode=::NewEnumerator1, UseLoopDelay=false)
   &SpawnBurst_Instantaneous(Spawn Count=2, Use Spawn Probability=false, UseLoopCountLimit=false)
   &SpawnBurst_Instantaneous(Spawn Count=4, Spawn Group=1, Use Spawn Probability=false, UseLoopCountLimit=false) {module=SpawnBurst_Instantaneous001}
   &SpawnBurst_Instantaneous(Spawn Count=30, Spawn Group=2, Use Spawn Probability=false, UseLoopCountLimit=false) {module=SpawnBurst_Instantaneous002}
   &SpawnRate(
    Spawn Probability=0.8
    SpawnGroup=3
    SpawnRate=&Lerp_Float(A=30, B=40, Alpha=$User.ParaGlideLeanAlpha)
    Use Spawn Probability=true
   )
  ]
  particleSpawnScript [
   /SetVariables(Particles.VisibilityTag=&SpawnGroup()) {isSetParameters=true}
   /SetVariables(Particles.Rand:NiagaraInt32=/Niagara/DynamicInputs/Arrays/SelectIntFromWeightedDistributionArray(
     Array Sampling Mode:ENiagara_LifetimeMode=::NewEnumerator1
     Direct Array Index:NiagaraInt32=?
     Direct Array Mode:ENiagara_WrapClamp=::NewEnumerator0
     Evaluation Type:ENiagaraRandomnessEvaluation=::NewEnumerator0
     Recalculate Random Each Loop:NiagaraBool=false
     Weighted Distribution Int Array:NiagaraDataInterfaceArrayDistributionInt=DataInterface() {propertyValues {ArrayData [{value=0, weight=0.3}, {value=1, weight=0.7}]}}
     Randomness Mode:ENiagaraRandomnessMode=::NewEnumerator0
     Random Seed:NiagaraInt32=0
     Fixed Random Seed:NiagaraBool=false
    )) {isSetParameters=true}
   &PartitionParticles()
   &InitializeParticle(
    Color=&MergeHelper_ColorSwitch(
     0=&FNTimeOfDayColor(Night Color=<rgba 2 1.3418 0 1>, Morning Color=<rgba 2 1.3418 0 1>, Day Color=<rgba 3 1.5852 0.1588 1>, Evening Color=<rgba 2 1.3418 0 1>)
     1=&MakeCustomLinearColorFromBool(
      Bool=&SetBoolByFloatComparison(A=$NPC.FortniteNPC.IsPlayerIndoors, B=0.5)
      True Color=<rgba 0.1476 0.0208 0 1>
      False Color=<rgba 3 0.5072 0 1>
     )
     2=<rgba 0 0 0 1>
     3=<rgba 4 2.2701 0 1>
     Index=&SpawnGroup()
     Num=::NewEnumerator2
    )
    Lifetime=&MergeHelper_FloatSwitch(
     0=5
     1=&MergeHelper_FloatSwitch(
      0=&RandomRangeFloat(Minimum=4, Maximum=5)
      1=1.8
      2=?
      3=?
      Index=$Particles.Rand
      Num=::NewEnumerator0
     )
     2=5
     3=&RandomRangeFloat(Minimum=0.2, Maximum=0.55)
     Index=&SpawnGroup()
     Num=::NewEnumerator2
    )
    Mesh Scale=&MergeHelper_VectorSwitch(
     0=<0.067 0.067 0.067>
     1=&Multiply_VectorByFloat(Vector=<0.05 0.09 0.09>, Float=&RandomRangeFloat(Minimum=2, Maximum=1.5))
     2=&Multiply_VectorByFloat(
      Vector=&RandomRangeVector(Minimum=<1.3 1.5 1.5>, Maximum=<1 1 1>)
      Float=&RandomRangeFloat(Minimum=0.05, Maximum=0.03)
     )
     3=&Multiply_VectorByFloat(
      Vector=&Lerp_Vector(
       A=&RandomRangeVector(Minimum=<0.7 0.4 0.4>, Maximum=<1 0.6 0.6>)
       B=&RandomRangeVector(Minimum=<1 1 1>, Maximum=<1 1 1>)
       Alpha=&FloatFromCurve(
        FloatCurve=DataInterface() {propertyValues {Curve {keys [{interpMode=RCIM_Cubic, tangentMode=RCTM_User, time=0.5, value=0}, <at 1 1>]}}}
        CurveIndex=&RandomRangeFloat(Minimum=0, Maximum=1)
       )
      )
      Float=0.3
     )
     Index=&SpawnGroup()
     Num=::NewEnumerator2
    )
    Mesh Scale Mode=::NewEnumerator4
    Sprite Size Mode=::NewEnumerator3
    UsePositionOffset=false
    Write Material Random=true
   )
   &ShapeLocation(
    Hemisphere Distribution=<1 0.5876>
    MaskSpawnGroup=true
    Offset=<0 0 25>
    Offset Mode=::NewEnumerator0
    Override Seed=false
    Rotation Axis=$Engine.Owner.SystemZAxis
    Rotation Mode=::NewEnumerator1
    Spawn Group Mask=2
    Sphere Radius=25
    Sphere Surface Distribution=0.4881
    UseOwnerScale=true
   )
   &InitialMeshOrientation(
    Mesh Orientation Mode=::NewEnumerator2
    Orientation Vector=&MergeHelper_VectorSwitch(
     0=<1 0 0>
     1=&SelectVectorFromArray(
      Array Sampling Mode=::NewEnumerator0
      Direct Array Index=?
      Vector Selection Array=DataInterface() {propertyValues {FloatData [<0 0 0>, <1 1 1>, <1 1 0>, <0 1 1>, <1 0 1>, <1 0 0>, <0 1 0>, <0 0 1>]}}
      Random Seed=0
     )
     2=$Particles.ShapeLocation.ShapeNormal
     3=&SelectVectorFromArray(
      Array Sampling Mode=::NewEnumerator0
      Direct Array Index=?
      Vector Selection Array=DataInterface() {propertyValues {FloatData [<0 0 0>, <1 1 1>, <1 1 0>, <0 1 1>, <1 0 1>, <1 0 0>, <0 1 0>, <0 0 1>]}}
      Random Seed=0
     )
     Index=&SpawnGroup()
     Num=::NewEnumerator2
    )
    Rotation=&MergeHelper_VectorSwitch(
     0=<0 0 0>
     1=<0 0 0>
     2=<0 0 0>
     3=&RandomRangeVector(Minimum=<-0.5 -0.5 -0.5>, Maximum=<0.5 0.5 0.5>)
     Index=&SpawnGroup()
     Num=::NewEnumerator2
    )
    Use Rotation Vector=true
   )
   /SetVariables(Particles.Random=&RandomRangeFloat(Minimum=0, Maximum=1)) {isSetParameters=true}
  ]
  particleUpdateScript [
   &ParticleState(Kill Particles When Lifetime Has Elapsed=false)
   &ScaleMeshSize(Scale Factor=&MergeHelper_VectorSwitch(
     0=<1 1 1>
     1=&VectorFromFloat(Value=&Lerp_Float(
       A=1
       B=&FloatFromCurve(
        FloatCurve=DataInterface() {propertyValues {Curve {
           defaultValue=1
           keys [
            {interpMode=RCIM_Cubic, tangentMode=RCTM_User, time=0, value=0.8}
            {interpMode=RCIM_Cubic, tangentMode=RCTM_User, time=0.5, value=1.2}
            {tangentMode=RCTM_User, time=1, value=0.8}
           ]
          }}}
        CurveIndex=&Frac_Float(Float=&Multiply_Float(A=$Engine.Time, B=3))
       )
       Alpha=$User.ParaGlideLeanAlpha
      ))
     2=<1 1 1>
     3=&Lerp_Vector(A=<1 1 0.3>, B=<1.5 1 1>, Alpha=$User.ParaGlideLeanAlpha)
     Index=&SpawnGroup()
     Num=::NewEnumerator2
    ))
   &ShapeLocation(
    MaskSpawnGroup=true
    Offset=<0 0 25>
    Offset Mode=::NewEnumerator0
    Override Seed=false
    Rotation Coordinate Space=::Simulation
    Spawn Group Mask=3
    Sphere Radius=5
    UseOwnerScale=true
   ) {module=ShapeLocation001}
   &SkeletalMeshLocation(
    Bone / Socket Index=$Particles.PartitionParticles.Partition
    MaskSpawnGroup=false
    Mesh Sampling Type=::NewEnumerator5
    Orientation Sampling=::NewEnumerator1
    Sampled Position Offset=&MergeHelper_VectorSwitch(
     0=<0 0 25>
     1=<0 0 25>
     2=Linked() {linkedVariable {name=Particles.ShapeLocation.ShapePosition, type=NiagaraPosition}}
     3=Linked() {linkedVariable {name=Particles.ShapeLocation001.ShapePosition, type=NiagaraPosition}}
     Index=&SpawnGroup()
     Num=::NewEnumerator2
    )
    Sampled Position Offset Coordinate Space:ENiagaraSkelMeshTransforms=::NewEnumerator2
    Skeletal Mesh=$System.Skeletal Mesh
    Socket Sampling Mode=::NewEnumerator5
   )
   &UpdateMeshOrientation(
    Rotation Rate=&MergeHelper_FloatSwitch(
     0=1.5
     1=&Multiply_Float(
      A=&MergeHelper_FloatSwitch(
       0=&RandomRangeFloat(Minimum=0.2, Maximum=0.15)
       1=0.5
       2=?
       3=?
       Index=$Particles.Rand
       Num=::NewEnumerator0
      )
      B=&FloatFromCurve(
       FloatCurve=DataInterface() {propertyValues {Curve {
          defaultValue=1
          keys [
           {interpMode=RCIM_Cubic, tangentMode=RCTM_User, time=0, value=0.5}
           {interpMode=RCIM_Cubic, tangentMode=RCTM_User, time=0.5, value=1}
           {tangentMode=RCTM_User, time=1, value=0.5}
          ]
         }}}
       CurveIndex=&Frac_Float(Float=&Add_Float(
         A=$Particles.NormalizedAge
         B=&MaskFloat(Mask Source=1, Mask Parameter=$Particles.PartitionParticles.Partition, Mask=0, Masked Float=0, Passthrough Float=0.5)
        ))
       Scale Curve=2.5
      )
     )
     2=&Lerp_Float(
      A=&RandomRangeFloat(Minimum=0.5, Maximum=0.2)
      B=&RandomRangeFloat(Minimum=2, Maximum=1)
      Alpha=$User.ParaGlideLeanAlpha
     )
     3=0
     Index=&SpawnGroup()
     Num=::NewEnumerator2
    )
    Rotation Vector=&MergeHelper_VectorSwitch(
     0=&SelectVectorFromArray(
      Direct Array Index=&ReturnExecIndex()
      Vector Selection Array=DataInterface() {propertyValues {FloatData [<-1 0.5 1>, <1 1 0.5>, <1 -1 1>, <-1 1 -1>, <-0.5 0.5 1>, <0.5 -1 1>]}}
     )
     1=<1 1 1>
     2=<1 0 0>
     3=<0 0 0>
     Index=&SpawnGroup()
     Num=::NewEnumerator2
    )
   )
   &ScaleColor(
    Scale Alpha=&MergeHelper_FloatSwitch(
     0=1
     1=1
     2=1
     3=&Multiply_Float(
      A=10
      B=&FloatFromCurve(FloatCurve=DataInterface() {propertyValues {Curve {keys [
           {interpMode=RCIM_Cubic, tangentMode=RCTM_User, time=0, value=0}
           {interpMode=RCIM_Cubic, tangentMode=RCTM_User, time=0.1, value=0}
           {interpMode=RCIM_Cubic, tangentMode=RCTM_User, time=0.2, value=0.8}
           {interpMode=RCIM_Cubic, tangentMode=RCTM_User, time=1, value=0}
          ]}}})
     )
     Index=&SpawnGroup()
     Num=::NewEnumerator2
    )
    Scale RGB=&MergeHelper_VectorSwitch(
     0=<1 1 1>
     1=<1 1 1>
     2=<1 1 1>
     3=&Multiply_VectorByFloat(
      Vector=&Multiply_VectorByFloat(Vector=<1 1 1>, Float=&FN_FrontEndFloatModifier(Gameplay=1, Front End=0.5))
      Float=&Multiply_Float(
       A=&RandomRangeFloat(Minimum=0.5, Maximum=1, Evaluation Type=::NewEnumerator1)
       B=&Multiply_Float(A=&FNTimeOfDayScalar(Night Scale=0.2, Morning Scale=1, Evening Scale=1), B=10)
      )
     )
     Index=&SpawnGroup()
     Num=::NewEnumerator2
    )
    ScaleA=true
    ScaleRGB=true
   )
   &DynamicMaterialParameters(
    Index 0 Param 1=&MergeHelper_FloatSwitch(
     0=1
     1=1
     2=1
     3=&FloatFromCurve(FloatCurve=DataInterface() {propertyValues {Curve {keys [{tangentMode=RCTM_User, time=0, value=0}, {tangentMode=RCTM_User, time=1, value=1}]}}})
     Index=&SpawnGroup()
     Num=::NewEnumerator2
    )
    Index 0 Param 3=&MergeHelper_FloatSwitch(
     0=1
     1=0
     2=0
     3=&FloatFromCurve(FloatCurve=DataInterface() {propertyValues {Curve {keys [
          {arriveTangent=4.0072, interpMode=RCIM_Cubic, leaveTangent=4.0072, tangentMode=RCTM_User, time=0, value=0}
          {arriveTangent=2.7083, interpMode=RCIM_Cubic, leaveTangent=2.7083, time=0.2543, value=0.6225}
          {arriveTangent=1.8473, interpMode=RCIM_Cubic, leaveTangent=1.8473, time=0.6, value=1.625}
          {arriveTangent=1.6993, interpMode=RCIM_Cubic, leaveTangent=1.6993, tangentMode=RCTM_User, time=1, value=2}
         ]}}})
     Index=&SpawnGroup()
     Num=::NewEnumerator2
    )
    Index 0 Param 4=20
    Index 1 Param 1=1
    Index 1 Param 2=$Particles.Random
    Index 1 Param 3=1.3
    Index 1 Param 4=1
    Index 2 Param 1=&FloatFromCurve(FloatCurve=DataInterface() {propertyValues {Curve {keys [<at 0 1>,<at 1 0.3>]}}})
    Module.Index0Float0Write=true
    Module.Index0Float1Write=false
    Module.Index0Float2Write=true
    Module.Index0Float3Write=true
    Module.Index1Float0Write=true
    Module.Index1Float1Write=true
    Module.Index1Float2Write=true
    Module.Index1Float3Write=true
    Module.Index2Float0Write=true
    Module.Index2Float1Write=true
    Module.Index2Float2Write=true
    Module.Index2Float3Write=true
    Write Parameter Index 1=::NewEnumerator1
    Write Parameter Index 2=::NewEnumerator1
   )
   /SetVariables(Particles.Position=&MergeHelper_VectorSwitch(
     0=Linked() {linkedVariable {name=StackContext.Position, type=NiagaraPosition}}
     1=Linked() {linkedVariable {name=StackContext.Position, type=NiagaraPosition}}
     2=&Lerp_Vector(
      A=Linked() {linkedVariable {name=StackContext.Position, type=NiagaraPosition}}
      B=Linked() {linkedVariable {name=Particles.SkeletalMeshLocation.SampledPosition, type=NiagaraPosition}}
      Alpha=&FloatFromCurve(
       FloatCurve=DataInterface() {propertyValues {Curve {
          defaultValue=1
          keys [
           {interpMode=RCIM_Cubic, tangentMode=RCTM_User, time=0, value=0}
           {interpMode=RCIM_Cubic, tangentMode=RCTM_User, time=0.5, value=0.3}
           {tangentMode=RCTM_User, time=1, value=0}
          ]
         }}}
       CurveIndex=&Frac_Float(Float=&Add_Float(
         A=&Multiply_Float(
          A=$Particles.NormalizedAge
          B=&MakeCustomFloatFromBool(
           Bool=&SetBoolByFloatComparison(A=$User.ParaGlideLeanAlpha, B=0.3)
           True Float=&RandomRangeFloat(Minimum=5, Maximum=10)
           False Float=&RandomRangeFloat(Minimum=3, Maximum=7)
          )
         )
         B=&RandomRangeFloat(Minimum=0, Maximum=0.7)
        ))
       Scale Curve=&RandomRangeFloat(Minimum=0.5, Maximum=1)
      )
     )
     3=Linked() {linkedVariable {name=StackContext.Position, type=NiagaraPosition}}
     Index=&SpawnGroup()
     Num=::NewEnumerator2
    )) {isSetParameters=true}
   /SetVariables(Particles.Lightning glow color:LinearColor=&MakeLinearColorFromVectorAndFloat(
     "Vector (RGB)"=&MakeVectorFromLinearColorRGB(Color=&MultiplyLinearColor(
       Linear Color=<rgba 4 1.9356 0 1>
       Scale Factor=&MakeVector4_FromFloat(Float=&RandomRangeFloat(Minimum=0.5, Maximum=1, Evaluation Type=::NewEnumerator1))
      ))
     "Float (Alpha)"=&Multiply_Float(
      A=&FloatFromCurve(FloatCurve=DataInterface() {propertyValues {Curve {keys [<at 0 1 RCIM_Cubic>, {interpMode=RCIM_Cubic, leaveTangent=-3.6307, time=0.9, value=0}]}}})
      B=5
     )
    )) {isSetParameters=true}
   /SetVariables(Particles.Glow Dy02:Vector4f=&MakeVector4(
     X=&FloatFromCurve(FloatCurve=DataInterface() {propertyValues {Curve {keys [{tangentMode=RCTM_User, time=0, value=0.2}, {tangentMode=RCTM_User, time=0.5, value=1}]}}})
     Y=0.7
     Z=0
     W=5
    )) {isSetParameters=true}
   /SetVariables(Particles.Glow Dy03:Vector4f=&MakeVector4(
     X=&FloatFromCurve(FloatCurve=DataInterface() {propertyValues {Curve {keys [<at 0 1>,<at 1 0.5>]}}}, Scale Curve=0.5)
     Y=$Particles.Random
     Z=1.3
     W=1
    )) {isSetParameters=true}
  ]
  rendererClasses [&NiagaraMeshRendererProperties]
  renderers [
   {rendererClass=&NiagaraMeshRendererProperties, rendererIndex=0}
   {rendererClass=&NiagaraMeshRendererProperties, rendererIndex=1}
   {rendererClass=&NiagaraMeshRendererProperties, rendererIndex=2}
   {rendererClass=&NiagaraMeshRendererProperties, rendererIndex=3}
   {rendererClass=&NiagaraMeshRendererProperties, rendererIndex=4}
  ]
 }
 Cubes_Filler {
  emitterUpdateScript [
   &EmitterState()
   &SpawnBurst_Instantaneous(Spawn Count=60, Use Spawn Probability=false, UseLoopCountLimit=false) {enabled=false}
   &SpawnRate(SpawnGroup=1, SpawnRate=40, Use Spawn Probability=false) {enabled=false, module=SpawnRate001}
   &SpawnRate(SpawnRate=15, Use Spawn Probability=false)
  ]
  particleSpawnScript [
   &FN_SetSpawnGroup(Number of Spawn Groups=2) {enabled=false}
   &PartitionParticles()
   &InitializeParticle(
    Color=&MakeCustomLinearColorFromBool(
     Bool=&Boolean_OrOperation(Bool1=$User.bIsFrontEnd, Bool2=$User.bIsFrontEndPreview)
     True Color=&RandomRangeLinearColor(Minimum=<rgba 1 0.1366 0 1>, Maximum=<rgba 1 0.3552 0.0155 1>)
     False Color=&FNTimeOfDayColor(
      Night Color=&RandomRangeLinearColor(Minimum=<rgba 1 0.2571 0.0793 1>, Maximum=<rgba 1 0.5263 0.1626 1>)
      Morning Color=&RandomRangeLinearColor(Minimum=<rgba 1 0.3132 0.1487 1>, Maximum=<rgba 1 0.5263 0.1626 1>)
      Day Color=&RandomRangeLinearColor(Minimum=<rgba 1 0.3132 0.1487 1>, Maximum=<rgba 1 0.5263 0.1626 1>)
      Evening Color=&RandomRangeLinearColor(Minimum=<rgba 1 0.3132 0.1487 1>, Maximum=<rgba 1 0.5263 0.1626 1>)
     )
    )
    Lifetime=0.8
    Mesh Scale Mode=::NewEnumerator3
    Mesh Uniform Scale=&Lerp_Float(
     A=&RandomRangeFloat(Minimum=0.2, Maximum=0.15)
     B=&RandomRangeFloat(Minimum=0.3, Maximum=0.1)
     Alpha=$User.ParaGlideLeanAlpha
    )
    Sprite Size=<5 20>
    Sprite Size Mode=::NewEnumerator4
    UsePositionOffset=false
    Write Material Random=true
   )
   &InitialMeshOrientation(
    Mesh Orientation Mode=::NewEnumerator1
    Rotation=&FN_FrontEndVector3Modifier(Gameplay=<0 0 0>, Front End=<0 0 0>)
    Use Rotation Vector=true
   )
  ]
  particleUpdateScript [
   &ParticleState()
   &ShapeLocation(
    Apply To Particle Position=::NewEnumerator1
    MaskSpawnGroup=false
    Offset=<0 0 30>
    Offset Mode=::NewEnumerator0
    Override Seed=false
    Shape Origin=&VectorFromFloat(Value=0)
    Sphere Radius=5
    Sphere Surface Distribution=0.6323
    UseOwnerScale=true
   )
   &ScaleMeshSize(Scale Factor=&VectorFromFloat(Value=&FloatFromCurve(
      FloatCurve=DataInterface() {propertyValues {Curve {keys [
          <at 0 0.7331 RCIM_Cubic>
          {arriveTangent=1.6648, interpMode=RCIM_Cubic, leaveTangent=1.6648, time=0.298, value=0.9303}
          {arriveTangent=-0.0612, interpMode=RCIM_Cubic, leaveTangent=-0.0612, time=0.5435, value=1.638}
          {arriveTangent=-4.3409, interpMode=RCIM_Cubic, leaveTangent=-4.3409, tangentMode=RCTM_User, time=0.7527, value=0.9024}
          <at 1 0 RCIM_Cubic>
         ]}}}
      Scale Curve=&FloatFromCurve(
       FloatCurve=DataInterface() {propertyValues {Curve {
          defaultValue=1
          keys [
           {arriveTangentWeight=83.3333, interpMode=RCIM_Cubic, leaveTangentWeight=83.3333, tangentMode=RCTM_User, time=16.5223, value=0}
           {arriveTangentWeight=83.3384, interpMode=RCIM_Cubic, leaveTangentWeight=83.3384, tangentMode=RCTM_User, time=169.9456, value=1}
          ]
         }}}
       CurveIndex=/Niagara/DynamicInputs/Camera/DistanceToCamera(Use Owner Position:NiagaraBool=true, Camera Query:NiagaraDataInterfaceCamera=DataInterface())
      )
     )))
   &UpdateMeshOrientation(Rotation Rate=&RandomRangeFloat(Minimum=-2, Maximum=2), Rotation Vector=<0 0 1>)
   &SkeletalMeshLocation(
    Bone / Socket Index=$Particles.PartitionParticles.Partition
    MaskSpawnGroup=false
    Mesh Sampling Type=::NewEnumerator5
    Orientation Sampling=::NewEnumerator1
    Sampled Position Offset=$Particles.ShapeLocation.ShapePosition
    Sampled Position Offset Coordinate Space:ENiagaraSkelMeshTransforms=::NewEnumerator2
    Skeletal Mesh=$System.Skeletal Mesh
    Socket Sampling Mode=::NewEnumerator5
   )
   /SetVariables(Particles.NewAge:NiagaraFloat=&Multiply_Float(A=$Particles.Age, B=0.5)) {isSetParameters=true, module=SetVariables_#3}
   /SetVariables(Particles.Velocity=&Multiply_VectorByFloat(
     Vector=&Lerp_Vector(
      A=<0 0 0>
      B=&FN_FrontEndVector3Modifier(Gameplay=$System.FN_Contrail_Setup.Direction Velocity, Front End=<0 -180 700>)
      Alpha=$Particles.NewAge
     )
     Float=1
    )) {isSetParameters=true, module=SetVariables_#4}
   /SetVariables(Particles.Position=&AddVectorToPosition(
     Position=$Particles.Position
     Vector=&Add_Vector(
      A=&Multiply_VectorByFloat(Vector=$Particles.Velocity, Float=$Particles.Age)
      B=&VectorFromCurve(
       VectorCurve=DataInterface() {propertyValues {
         XCurve {
          defaultValue=0
          keys [
           {interpMode=RCIM_Cubic, tangentMode=RCTM_User, time=0, value=0}
           {interpMode=RCIM_Cubic, tangentMode=RCTM_User, time=0.5, value=1}
           <at 1 -1 RCIM_Cubic>
          ]
         }
         YCurve {
          defaultValue=0
          keys [
           {interpMode=RCIM_Cubic, tangentMode=RCTM_User, time=0, value=0}
           {interpMode=RCIM_Cubic, tangentMode=RCTM_User, time=0.5, value=1}
           {tangentMode=RCTM_User, time=1, value=0}
          ]
         }
         ZCurve {keys [<at 0 0>,<at 1 0>]}
        }}
       Scale Curve=&MaskVector(Mask Source=1, Mask Parameter=$Particles.PartitionParticles.Partition, Mask=0, Masked Vector=<5 1 1>, Passthrough Vector=<-5 1 1>)
      )
     )
    )) {isSetParameters=true, module=SetVariables_#5}
   &ScaleColor(
    Scale Alpha=&FloatFromCurve(
     FloatCurve=DataInterface() {propertyValues {Curve {keys [
         {interpMode=RCIM_Cubic, tangentMode=RCTM_User, time=0, value=0}
         {interpMode=RCIM_Cubic, tangentMode=RCTM_User, time=0.5, value=1}
         {tangentMode=RCTM_User, time=1, value=0}
        ]}}}
     Scale Curve=$System.FN_Contrail_Setup.Normalized Intro Age
    )
    Scale RGB=&Multiply_Vector(
     A=&VectorFromFloat(Value=&FloatFromCurve(FloatCurve=DataInterface() {propertyValues {Curve {keys [<at 0 1.2913>,<at 0.4993 1.1767>,<at 1 1.1148>]}}}))
     B=&MakeVectorFromLinearColorRGB(Color=&LerpLinearColors(
       StartColor=&FNTimeOfDayColor(
        Night Color=<rgba 2.8212 3 0 1>
        Morning Color=<rgba 1.9333 2 0.8808 1>
        Day Color=<rgba 1.9333 2 0.8808 1>
        Evening Color=<rgba 1.9333 2 0.8808 1>
       )
       EndColor=<rgba 1 1 1 1>
       LerpFactor=&FloatFromCurve(FloatCurve=DataInterface() {propertyValues {Curve {
           defaultValue=1
           keys [{tangentMode=RCTM_User, time=0, value=0}, <at 0.2364 0.6637>, <at 0.4691 0.933>, {tangentMode=RCTM_User, time=1, value=1}]
          }}})
      ))
    )
    ScaleA=false
    ScaleRGB=true
   )
  ]
  rendererClasses [&NiagaraMeshRendererProperties]
  renderers [{rendererClass=&NiagaraMeshRendererProperties, rendererIndex=0}]
 }
 Cubes_Trail {
  emitterUpdateScript [
   &EmitterState()
   &SpawnRate(SpawnGroup=1, SpawnRate=40, Use Spawn Probability=false) {enabled=false, module=SpawnRate001}
   &SpawnRate(SpawnRate=100, Use Spawn Probability=false)
  ]
  particleSpawnScript [
   &FN_SetSpawnGroup(Number of Spawn Groups=2)
   &PartitionParticles(Particles Per Partition=2, Partition Mode:ENiagara_PartitionMode=::NewEnumerator1)
   &InitializeParticle(
    Color=&FNTimeOfDayColor(
     Night Color=&RandomRangeLinearColor(Minimum=<rgba 1 0.6709 0 1>, Maximum=<rgba 1 0.5936 0 1>)
     Morning Color=&RandomRangeLinearColor(Minimum=<rgba 1 0.6709 0 1>, Maximum=<rgba 1 0.5936 0 1>)
     Day Color=&RandomRangeLinearColor(Minimum=<rgba 1 0.6709 0 1>, Maximum=<rgba 1 0.5936 0 1>)
     Evening Color=&RandomRangeLinearColor(Minimum=<rgba 1 0.6709 0 1>, Maximum=<rgba 1 0.5936 0 1>)
    )
    Lifetime Max=0.6
    Lifetime Min=0.8
    Lifetime Mode=::NewEnumerator1
    Mesh Scale Mode=::NewEnumerator3
    Mesh Uniform Scale=&RandomRangeFloat(Minimum=0.02, Maximum=0.03)
    Sprite Size=<5 20>
    Sprite Size Mode=::NewEnumerator4
    UsePositionOffset=false
    Write Material Random=true
   )
   &InitialMeshOrientation(Use Rotation Vector=false)
  ]
  particleUpdateScript [
   &ParticleState()
   &ShapeLocation(
    Apply To Particle Position=::NewEnumerator1
    MaskSpawnGroup=false
    Offset=<0 0 30>
    Offset Mode=::NewEnumerator0
    Override Seed=false
    Shape Origin=&VectorFromFloat(Value=0)
    Sphere Radius=&Lerp_Float(
     A=11
     B=40
     Alpha=&FloatFromCurve(FloatCurve=DataInterface() {propertyValues {Curve {
         defaultValue=0
         keys [
          {interpMode=RCIM_Cubic, tangentMode=RCTM_User, time=0, value=0.1554}
          {arriveTangent=0.8446, interpMode=RCIM_Cubic, leaveTangent=0.8446, time=0.6564, value=0.318}
          {arriveTangent=1.9692, interpMode=RCIM_Cubic, leaveTangent=1.9692, tangentMode=RCTM_User, time=1, value=1}
         ]
        }}})
    )
    Sphere Surface Distribution=1
    UseOwnerScale=true
   )
   &ScaleMeshSize(Scale Factor=&VectorFromFloat(Value=&Multiply_Float(
      A=&FloatFromCurve(
       FloatCurve=DataInterface() {propertyValues {Curve {
          defaultValue=1
          keys [
           {arriveTangent=6.8383, interpMode=RCIM_Cubic, leaveTangent=6.8383, tangentMode=RCTM_User, time=0, value=0}
           {interpMode=RCIM_Cubic, tangentMode=RCTM_User, time=0.5, value=1}
           {arriveTangent=-7.0617, tangentMode=RCTM_User, time=1, value=0}
          ]
         }}}
       Scale Curve=/Niagara/DynamicInputs/Helpers/DistributionCurveFloat(
        Minimum:NiagaraFloat=0.3
        Maximum:NiagaraFloat=1
        Distribution Curve:NiagaraDataInterfaceCurve=DataInterface() {propertyValues {Curve {
           defaultValue=0
           keys [
            {arriveTangent=1.0827, interpMode=RCIM_Cubic, leaveTangent=1.0827, tangentMode=RCTM_User, time=0, value=0.186}
            {interpMode=RCIM_Cubic, tangentMode=RCTM_User, time=1, value=1}
           ]
          }}}
        Curve Parameter Type:NiagaraInt32=0
        "Curve Parameter Type:":NiagaraInt32=0
        Curve Parameter:NiagaraFloat=?
        Evaluation Type:ENiagaraRandomnessEvaluation=::NewEnumerator0
        Recalculate Random Each Loop:NiagaraBool=true
        Randomness Mode:ENiagaraRandomnessMode=::NewEnumerator0
        Random Seed:NiagaraInt32=0
        Fixed Random Seed:NiagaraBool=false
       )
      )
      B=&FloatFromCurve(
       FloatCurve=DataInterface() {propertyValues {Curve {
          defaultValue=1
          keys [
           {arriveTangentWeight=83.3333, interpMode=RCIM_Cubic, leaveTangentWeight=83.3333, tangentMode=RCTM_User, time=39.1423, value=0}
           {arriveTangentWeight=83.3384, interpMode=RCIM_Cubic, leaveTangentWeight=83.3384, tangentMode=RCTM_User, time=123.2398, value=1}
          ]
         }}}
       CurveIndex=&DistanceToCamera(Use Owner Position:NiagaraBool=true, Camera Query:NiagaraDataInterfaceCamera=DataInterface())
      )
     )))
   &UpdateMeshOrientation(Rotation Rate=&RandomRangeFloat(Minimum=-2, Maximum=2), Rotation Vector=<1 1 1>)
   &SkeletalMeshLocation(
    Bone / Socket Index=$Particles.PartitionParticles.Partition
    MaskSpawnGroup=false
    Mesh Sampling Type=::NewEnumerator5
    Orientation Sampling=::NewEnumerator1
    Sampled Position Offset=$Particles.ShapeLocation.ShapePosition
    Sampled Position Offset Coordinate Space:ENiagaraSkelMeshTransforms=::NewEnumerator2
    Skeletal Mesh=$System.Skeletal Mesh
    Socket Sampling Mode=::NewEnumerator5
   )
   /SetVariables(Particles.NewAge:NiagaraFloat=&Multiply_Float(A=$Particles.Age, B=0.75)) {isSetParameters=true, module=SetVariables_#3}
   /SetVariables(Particles.Velocity=&Multiply_VectorByFloat(
     Vector=&Lerp_Vector(
      A=<0 0 0>
      B=&FN_FrontEndVector3Modifier(Gameplay=$System.FN_Contrail_Setup.Direction Velocity, Front End=<0 -100 500>)
      Alpha=$Particles.NewAge
     )
     Float=1
    )) {isSetParameters=true, module=SetVariables_#4}
   /SetVariables(Particles.Position=&AddVectorToPosition(
     Position=$Particles.Position
     Vector=&Add_Vector(
      A=&Multiply_VectorByFloat(Vector=$Particles.Velocity, Float=$Particles.Age)
      B=&VectorFromCurve(
       VectorCurve=DataInterface() {propertyValues {
         XCurve {
          defaultValue=0
          keys [
           {interpMode=RCIM_Cubic, tangentMode=RCTM_User, time=0, value=0}
           {interpMode=RCIM_Cubic, tangentMode=RCTM_User, time=0.5, value=1}
           <at 1 -1 RCIM_Cubic>
          ]
         }
         YCurve {
          defaultValue=0
          keys [
           {interpMode=RCIM_Cubic, tangentMode=RCTM_User, time=0, value=0}
           {interpMode=RCIM_Cubic, tangentMode=RCTM_User, time=0.5, value=1}
           {tangentMode=RCTM_User, time=1, value=0}
          ]
         }
         ZCurve {keys [<at 0 0>,<at 1 0>]}
        }}
       Scale Curve=&MaskVector(Mask Source=1, Mask Parameter=$Particles.PartitionParticles.Partition, Mask=0, Masked Vector=<5 1 1>, Passthrough Vector=<-5 1 1>)
      )
     )
    )) {isSetParameters=true, module=SetVariables_#5}
   &ScaleColor(
    Scale Alpha=&FloatFromCurve(
     FloatCurve=DataInterface() {propertyValues {Curve {keys [
         {interpMode=RCIM_Cubic, tangentMode=RCTM_User, time=0, value=0}
         {interpMode=RCIM_Cubic, tangentMode=RCTM_User, time=0.5, value=1}
         {tangentMode=RCTM_User, time=1, value=0}
        ]}}}
     Scale Curve=$System.FN_Contrail_Setup.Normalized Intro Age
    )
    Scale RGB=&VectorFromFloat(Value=&FloatFromCurve(FloatCurve=DataInterface() {propertyValues {Curve {keys [<at 0 1.2642>,<at 0.4993 0.9559>,<at 1 0.7492>]}}}))
    ScaleA=false
    ScaleRGB=true
   )
  ]
  rendererClasses [&NiagaraMeshRendererProperties]
  renderers [{rendererClass=&NiagaraMeshRendererProperties, rendererIndex=0}]
 }
 Glows {
  emitterUpdateScript [
   &EmitterState(Life Cycle Mode=::NewEnumerator1, Loop Behavior=::NewEnumerator1, Loop Duration Mode=::NewEnumerator1, UseLoopDelay=false)
   &SpawnBurst_Instantaneous(Spawn Count=2, Use Spawn Probability=false, UseLoopCountLimit=true)
  ]
  particleSpawnScript [
   &PartitionParticles()
   &InitializeParticle(
    Color=<rgba 0.3229 0.1646 0.0165 1>
    Lifetime=2
    Mesh Scale Mode=::NewEnumerator3
    Mesh Uniform Scale=0.46
    UsePositionOffset=false
    Write Material Random=true
   )
  ]
  particleUpdateScript [
   &ParticleState(Kill Particles When Lifetime Has Elapsed=false)
   /SetVariables(Particles.OutlineColor:LinearColor=&Multiply_LinearColorByFloat(
     Linear Color=<rgba 4 4 4 1>
     Float=&FloatFromCurve(
      FloatCurve=DataInterface() {propertyValues {Curve {
         defaultValue=1
         keys [
          {interpMode=RCIM_Cubic, tangentMode=RCTM_User, time=0, value=1}
          {interpMode=RCIM_Cubic, tangentMode=RCTM_User, time=0.5, value=2}
          {tangentMode=RCTM_User, time=1, value=1}
         ]
        }}}
      Scale Curve=&FloatFromCurve(
       FloatCurve=DataInterface() {propertyValues {Curve {keys [{interpMode=RCIM_Cubic, tangentMode=RCTM_User, time=0, value=0}, <at 1 1>]}}}
       CurveIndex=$Emitter.Age
      )
     )
    )) {isSetParameters=true}
   &SkeletalMeshLocation(
    Bone / Socket Index=$Particles.PartitionParticles.Partition
    MaskSpawnGroup=false
    Mesh Sampling Type=::NewEnumerator5
    Orientation Sampling=::NewEnumerator1
    Sampled Position Offset=<0 0 25>
    Sampled Position Offset Coordinate Space:ENiagaraSkelMeshTransforms=::NewEnumerator2
    Skeletal Mesh=$System.Skeletal Mesh
    Socket Sampling Mode=::NewEnumerator5
   )
   /SetVariables(
    Particles.Glow_Base_CameraOffset=-5
    Particles.Glow_Base_Color:LinearColor=&Multiply_LinearColorByFloat(
     Linear Color=&MakeCustomLinearColorFromBool(
      Bool=&Boolean_OrOperation(Bool1=$User.bIsFrontEnd, Bool2=$User.bIsFrontEndPreview)
      True Color=<rgba 1 0.3925 0 1>
      False Color=<rgba 3 1.1775 0 1>
     )
     Float=&MaskFloat(
      Mask Source=1
      Mask Parameter=$Particles.PartitionParticles.Partition
      Mask=0
      Masked Float=&FloatFromCurve(
       FloatCurve=DataInterface() {propertyValues {Curve {keys [<at 0 1>, {arriveTangent=1.2666, leaveTangent=1.2666, time=1.5, value=1}, <at 1.5791 3>, <at 1.7 1>]}}}
       CurveIndex=$Emitter.Age
      )
      Passthrough Float=1
     )
    )
    Particles.Glow_Base_Size:Vector2f=&Multiply_Vector2D_ByFloat(Float=3, Vector2D=<22 22>)
   ) {isSetParameters=true}
   &Light_Attributes(
    Color=1
    Color Value To Scale:LinearColor=/Niagara/DynamicInputs/LinearColor/MakeLinearColorFromVector4(Vector:Vector4f=<1 1 1 1>)
    Falloff Exponent=1
    Light Exponent=2
    Light Position Offset:Vector3f=<0 -8 0>
    Light Radius=40
    Position=1
    Position Offset Coordinate Space:ENiagaraCoordinateSpace=::Local
    Radius=1
    Scale Alpha=40
    Scale RGB:Vector3f=&MakeVectorFromLinearColorRGB(Color=<rgba 1 0.3713 0 1>)
    ScaleA=true
    ScaleRGB=true
    UsePositionOffset=true
   )
  ]
  rendererClasses [&NiagaraSpriteRendererProperties, &NiagaraLightRendererProperties]
  renderers [
   {rendererClass=&NiagaraSpriteRendererProperties}
   {rendererClass=&NiagaraLightRendererProperties, rendererIndex=1}
  ]
 }
 Lightning {
  emitterUpdateScript [
   &EmitterState(
    Life Cycle Mode=::NewEnumerator1
    Loop Duration=&RandomRangeFloat(Minimum=0.25, Maximum=3.5)
    Recalculate Duration Each Loop=true
    UseLoopDelay=false
   )
   &SpawnRate(SpawnRate=&Lerp_Float(A=15, B=30, Alpha=$User.ParaGlideLeanAlpha), Use Spawn Probability=true)
  ]
  particleSpawnScript [
   &PartitionParticles()
   &InitializeParticle(
    Color=&RandomRangeLinearColor(Minimum=<rgba 20 4.3087 0 1>, Maximum=<rgba 10 0.6222 0 1>)
    Lifetime=&RandomRangeFloat(Minimum=0.15, Maximum=0.4)
    Mesh Scale=&Multiply_VectorByFloat(Vector=&RandomRangeVector(Minimum=<0.2 0.2 0.8>, Maximum=<1 1 1.5>), Float=10)
    Mesh Scale Mode=::NewEnumerator4
    Sprite Size Mode=::NewEnumerator3
    UsePositionOffset=false
    Write Material Random=true
   )
   &InitialMeshOrientation(
    Mesh Orientation Mode=::NewEnumerator2
    Orientation Vector=&FN_FrontEndVector3Modifier(Gameplay=<1 0 0>, Front End=<1 0 0>)
    Rotation=&RandomRangeVector(Minimum=<-0.035 -0.035 0>, Maximum=<0.035 0.035 1>)
    Use Rotation Vector=true
   )
   &CameraOffset(Camera Offset Amount=20)
   &ShapeLocation(
    MaskSpawnGroup=false
    Non Uniform Scale=<1 1 -0.5>
    Offset=&RandomRangeVector(Minimum=<0 0 -50>, Maximum=<0 0 -75>)
    Offset Mode=::NewEnumerator0
    Override Seed=false
    Ring Radius=50
    Shape Primitive=::NewEnumerator4
    UseOwnerScale=true
   ) {enabled=false}
   &InitialMeshOrientation(
    Mesh Axis To Orient=<0 0 1>
    Mesh Orientation Mode=::NewEnumerator2
    Orientation Vector=&FN_FrontEndVector3Modifier(Gameplay=<0 0 1>, Front End=<0 -0.2 1>)
    Use Rotation Vector=false
   ) {module=InitialMeshOrientation001}
  ]
  particleUpdateScript [
   &ParticleState()
   &DynamicMaterialParameters(
    Index 0 Param 1=&FloatFromCurve(
     FloatCurve=DataInterface() {propertyValues {Curve {keys [
         {arriveTangent=3.0269, interpMode=RCIM_Cubic, leaveTangent=3.0269, tangentMode=RCTM_User, time=0, value=0}
         {tangentMode=RCTM_User, time=1, value=1}
        ]}}}
     Scale Curve=&RandomRangeFloat(Minimum=0, Maximum=1)
    )
    Index 0 Param 2=0.7
    Index 0 Param 3=&FloatFromCurve(FloatCurve=DataInterface() {propertyValues {Curve {keys [<at 0.25 0 RCIM_Cubic>, {arriveTangent=5.448, interpMode=RCIM_Cubic, leaveTangent=5.448, tangentMode=RCTM_User, time=1, value=3}]}}})
    Index 0 Param 4=&FloatFromCurve(
     FloatCurve=DataInterface() {propertyValues {Curve {keys [
         {arriveTangent=2.5109, interpMode=RCIM_Cubic, leaveTangent=2.5109, tangentMode=RCTM_User, time=0, value=0}
         {interpMode=RCIM_Cubic, tangentMode=RCTM_User, time=1, value=1}
        ]}}}
     Scale Curve=&RandomRangeFloat(Minimum=0, Maximum=1)
    )
    Index 1 Param 1=&FloatFromCurve(
     FloatCurve=DataInterface() {propertyValues {Curve {keys [<at 0.25 1>,<at 1 0>]}}}
     Scale Curve=&RandomRangeFloat(Minimum=1, Maximum=3)
    )
    Index 1 Param 2=&RandomRangeFloat(Minimum=0, Maximum=1)
    Index 1 Param 3=&RandomRangeFloat(Minimum=0.5, Maximum=1)
    Index 1 Param 4=1
    Index 2 Param 1=&FloatFromCurve(FloatCurve=DataInterface() {propertyValues {Curve {keys [{arriveTangent=-2.6239, interpMode=RCIM_Cubic, leaveTangent=-2.6239, tangentMode=RCTM_User, time=0, value=1}, <at 1 0.3>]}}})
    Module.Param0WriteEnabled=true
    Module.Param0WriteEnabled1=true
    Module.Param0WriteEnabled2=true
    Module.Param1WriteEnabled=true
    Module.Param1WriteEnabled1=true
    Module.Param1WriteEnabled2=true
    Module.Param2WriteEnabled=true
    Module.Param2WriteEnabled1=true
    Module.Param2WriteEnabled2=true
    Module.Param3WriteEnabled=true
    Module.Param3WriteEnabled1=true
    Module.Param3WriteEnabled2=true
    Write Parameter Index 1=true
    Write Parameter Index 2=true
   )
   &ShapeLocation(
    Apply To Particle Position=::NewEnumerator1
    MaskSpawnGroup=false
    Offset Mode=::NewEnumerator0
    Override Seed=false
    Shape Origin=&VectorFromFloat(Value=0)
    Sphere Radius=7
    Sphere Surface Distribution=1
    UseOwnerScale=true
   ) {module=ShapeLocation001}
   &SkeletalMeshLocation(
    Bone / Socket Index=$Particles.PartitionParticles.Partition
    MaskSpawnGroup=false
    Mesh Sampling Type=::NewEnumerator5
    Orientation Sampling=::NewEnumerator1
    Sampled Position Offset:NiagaraPosition=$StackContext.ShapeLocation001.ShapePosition
    Sampled Position Offset Coordinate Space:ENiagaraSkelMeshTransforms=::NewEnumerator2
    Skeletal Mesh=$System.Skeletal Mesh
    Socket Sampling Mode=::NewEnumerator5
   )
   &ScaleMeshSize(Scale Factor=&VectorFromCurve(VectorCurve=DataInterface() {propertyValues {XCurve {keys [<at 0 1>,<at 1 1>]}, YCurve {keys [<at 0 1>,<at 1 1>]}, ZCurve {keys [<at 0 0.2>,<at 1 1.474>]}}}))
   &ScaleColor(
    Scale Alpha=&Multiply_Float(
     A=1
     B=&FloatFromCurve(FloatCurve=DataInterface() {propertyValues {Curve {keys [<at 0 0>,<at 0.1 0>,<at 0.2 0.8>,<at 1 0>]}}})
    )
    Scale RGB=&Multiply_VectorByFloat(
     Vector=<1 1 1>
     Float=&FloatFromCurve(
      FloatCurve=DataInterface() {propertyValues {Curve {keys [
          {arriveTangent=4.9531, interpMode=RCIM_Cubic, leaveTangent=4.9531, tangentMode=RCTM_User, time=0, value=1}
          {arriveTangent=-2.0019, interpMode=RCIM_Cubic, leaveTangent=-2.0019, time=0.05, value=10}
          {arriveTangent=-0.2096, interpMode=RCIM_Cubic, leaveTangent=-0.2096, tangentMode=RCTM_User, time=0.1983, value=0.6029}
          <at 1 0>
         ]}}}
      Scale Curve=&DistributionCurveFloat(
       Minimum:NiagaraFloat=1
       Maximum:NiagaraFloat=3
       Distribution Curve:NiagaraDataInterfaceCurve=DataInterface() {propertyValues {Curve {
          defaultValue=0
          keys [
           {interpMode=RCIM_Cubic, tangentMode=RCTM_User, time=0, value=0}
           {arriveTangent=2.8548, interpMode=RCIM_Cubic, leaveTangent=2.8548, tangentMode=RCTM_User, time=1, value=1}
          ]
         }}}
       Curve Parameter Type:NiagaraInt32=0
       "Curve Parameter Type:":NiagaraInt32=0
       Curve Parameter:NiagaraFloat=?
       Evaluation Type:ENiagaraRandomnessEvaluation=::NewEnumerator0
       Recalculate Random Each Loop:NiagaraBool=true
       Randomness Mode:ENiagaraRandomnessMode=::NewEnumerator0
       Random Seed:NiagaraInt32=0
       Fixed Random Seed:NiagaraBool=false
      )
     )
    )
    ScaleA=true
    ScaleRGB=true
   )
   &SolveForcesAndVelocity(Clamp Velocity=false, Limit Acceleration=false)
  ]
  rendererClasses [&NiagaraMeshRendererProperties]
  renderers [{rendererClass=&NiagaraMeshRendererProperties, rendererIndex=0}]
 }
 RockShards_Tail {
  emitterUpdateScript [
   &EmitterState(Delay First Loop Only=true, Life Cycle Mode=::NewEnumerator1, Loop Delay=0.2416, Loop Duration=1.2584, UseLoopDelay=true)
   &SpawnRate(SpawnRate=100, Use Spawn Probability=true)
  ]
  particleSpawnScript [
   &PartitionParticles()
   &InitializeParticle(
    Color=<rgba 0.3281 0.3281 0.3281 1>
    Lifetime Max=0.3
    Lifetime Min=0.4
    Lifetime Mode=::NewEnumerator1
    Mesh Renderer Array Visibility Mode=::NewEnumerator3
    Mesh Renderer Info=DataInterface() {propertyValues {MeshRenderer=&self:RockShards_Tail_6.NiagaraMeshRendererProperties_1}}
    Mesh Scale=&Multiply_VectorByFloat(
     Vector=&MakeVector(
      X=&RandomRangeFloat(Minimum=0.35, Maximum=0.75)
      Y=&RandomRangeFloat(Minimum=0.35, Maximum=0.75)
      Z=&RandomRangeFloat(Minimum=0.5, Maximum=1)
     )
     Float=&RandomRangeFloat(Minimum=0.8, Maximum=1)
    )
    Mesh Scale Mode=::NewEnumerator4
    UsePositionOffset=true
    Write Material Random=true
   )
   &ShapeLocation(
    Apply To Particle Position=::NewEnumerator1
    MaskSpawnGroup=false
    Offset=<0 0 30>
    Offset Mode=::NewEnumerator0
    Override Seed=false
    Shape Origin=&VectorFromFloat(Value=0)
    Sphere Radius=&RandomRangeFloat(Minimum=0, Maximum=20)
    Sphere Surface Distribution=1
    UseOwnerScale=true
   )
   &SkeletalMeshLocation(
    Bone / Socket Index=$Particles.PartitionParticles.Partition
    Interpolate Mesh Sampling=false
    MaskSpawnGroup=false
    Mesh Sampling Type=::NewEnumerator5
    Orientation Sampling=::NewEnumerator1
    Sampled Position Offset=$StackContext.ShapeLocation.ShapePosition
    Sampled Position Offset Coordinate Space:ENiagaraSkelMeshTransforms=::NewEnumerator2
    Skeletal Mesh=$System.Skeletal Mesh
    Socket Sampling Mode=::NewEnumerator5
   )
   &AddVelocity(
    Velocity=&FN_FrontEndVector3Modifier(Gameplay=<0 0 1>, Front End=<0 -0.3 1>)
    Velocity Speed Scale=&RandomRangeFloat(Minimum=500, Maximum=600)
   ) {module=AddVelocity001}
   &InitialMeshOrientation(Use Rotation Vector=false)
  ]
  particleUpdateScript [
   &ParticleState()
   &ScaleMeshSize(Scale Factor=&Multiply_VectorByFloat(
     Vector=&Lerp_Vector(
      A=<0.8 0.8 0.8>
      B=<1.2 1.2 1.2>
      Alpha=&FloatFromCurve(
       FloatCurve=DataInterface() {propertyValues {Curve {
          defaultValue=0.9971
          keys [
           <at 0.0046 0 RCIM_Cubic>
           {arriveTangentWeight=0.0081, interpMode=RCIM_Cubic, tangentMode=RCTM_User, time=0.0288, value=1}
           {arriveTangent=0.038, arriveTangentWeight=0.1237, interpMode=RCIM_Cubic, leaveTangent=0.038, tangentMode=RCTM_User, time=0.4, value=0.95}
           {arriveTangent=-2.0607, arriveTangentWeight=0.1667, interpMode=RCIM_Cubic, leaveTangent=-2.0607, tangentMode=RCTM_User, time=0.9, value=0}
          ]
         }}}
       Scale Curve=0.4
      )
     )
     Float=&FloatFromCurve(
      FloatCurve=DataInterface() {propertyValues {Curve {keys [
          {arriveTangent=20.3126, interpMode=RCIM_Cubic, leaveTangent=20.3126, tangentMode=RCTM_User, time=0, value=0.25}
          {interpMode=RCIM_Cubic, tangentMode=RCTM_User, time=0.0541, value=1.2}
          {arriveTangent=-0.0633, interpMode=RCIM_Cubic, leaveTangent=-0.0633, time=0.2338, value=0.85}
          {interpMode=RCIM_Cubic, leaveTangent=-1.1083, tangentMode=RCTM_Break, time=0.561, value=1.1679}
          {arriveTangent=-0.3747, interpMode=RCIM_Cubic, leaveTangent=-0.3747, tangentMode=RCTM_User, time=1, value=0.25}
         ]}}}
      Scale Curve=&FloatFromCurve(
       FloatCurve=DataInterface() {propertyValues {Curve {
          defaultValue=1
          keys [
           {arriveTangentWeight=83.3333, interpMode=RCIM_Cubic, leaveTangentWeight=83.3333, tangentMode=RCTM_User, time=39.1423, value=0}
           {arriveTangentWeight=83.3384, interpMode=RCIM_Cubic, leaveTangentWeight=83.3384, tangentMode=RCTM_User, time=123.2398, value=1}
          ]
         }}}
       CurveIndex=&DistanceToCamera(Use Owner Position:NiagaraBool=true, Camera Query:NiagaraDataInterfaceCamera=DataInterface())
      )
     )
    ))
   &MeshRotationForce(Rotation=&RandomRangeVector(Minimum=<-20 -20 -20>, Maximum=<20 20 20>))
   &SolveForcesAndVelocity(Clamp Velocity=false, Limit Acceleration=false)
  ]
  rendererClasses [&NiagaraMeshRendererProperties]
  renderers [{rendererClass=&NiagaraMeshRendererProperties, rendererIndex=0}]
 }
 TrailLong {
  emitterUpdateScript [
   &EmitterState(
    Inactive Response=::NewEnumerator1
    Life Cycle Mode=::NewEnumerator1
    Loop Behavior=::NewEnumerator1
    Loop Duration Mode=::NewEnumerator1
    UseLoopDelay=false
   )
   &SpawnBurst_Instantaneous(Spawn Count=2, Use Spawn Probability=false, UseLoopCountLimit=false)
   /SetVariables(Emitter.Texture_Speed_Master:NiagaraFloat=&MakeCustomFloatFromBool(
     Bool=&SetBoolByFloatComparison(A=1, B=$User.ParaGlideLeanAlpha, Comparison Type=::NewEnumerator4)
     True Float=0.7
     False Float=&RandomRangeFloat(Minimum=0.4, Maximum=0.5)
    )) {isSetParameters=true}
   /SetVariables(Emitter.Texture_Speed_Master_Detail:NiagaraFloat=&MakeCustomFloatFromBool(
     Bool=&SetBoolByFloatComparison(A=1, B=$User.ParaGlideLeanAlpha, Comparison Type=::NewEnumerator4)
     True Float=0.6
     False Float=0.45
    )) {isSetParameters=true}
   /SetVariables(Emitter.SphereMaskU:NiagaraFloat=&Lerp_Float(A=1, B=0.95, Alpha=$User.ParaGlideLeanAlpha)) {isSetParameters=true}
  ]
  particleSpawnScript [
   &PartitionParticles()
   &InitializeParticle(
    Color=&FNTimeOfDayColor(
     Night Color=<rgba 1 0.4875 0 0.6502>
     Morning Color=<rgba 0.7726 0.4475 0.0015 0.5188>
     Day Color=<rgba 0.7726 0.4475 0.0015 0.5188>
     Evening Color=<rgba 0.7726 0.4475 0.0015 0.5188>
    )
    Mesh Scale=<1.1 0.7 1>
    Mesh Scale Mode=::NewEnumerator4
    UsePositionOffset=true
    Write Material Random=true
   )
   &InitialMeshOrientation(Mesh Orientation Mode=::NewEnumerator2, Rotation=<0.25 0 0>, Use Rotation Vector=true)
   &InitialMeshOrientation(
    Mesh Orientation Mode=::NewEnumerator2
    Rotation=&SelectVectorFromArray(
     Direct Array Index=$Particles.PartitionParticles.Partition
     Vector Selection Array=DataInterface() {propertyValues {FloatData [<0 0 0>, <0 0 0.1>]}}
    )
    Use Rotation Vector=true
   ) {module=InitialMeshOrientation001}
  ]
  particleUpdateScript [
   &ParticleState(Kill Particles When Lifetime Has Elapsed=false, Let Infinitely Lived Particles Die When Emitter Deactivates=true)
   &ScaleMeshSize(Scale Factor=&VectorFromCurve(
     VectorCurve=DataInterface() {propertyValues {
       XCurve {defaultValue=0, keys [{tangentMode=RCTM_User, time=0, value=0}, {tangentMode=RCTM_User, time=1, value=1}]}
       YCurve {defaultValue=0, keys [{tangentMode=RCTM_User, time=0, value=0}, {tangentMode=RCTM_User, time=1, value=1}]}
       ZCurve {defaultValue=0, keys [{tangentMode=RCTM_User, time=0, value=0}, {tangentMode=RCTM_User, time=1, value=1}]}
      }}
     Scale Curve=&MakeCustomVectorFromBool(
      Bool=$User.bIsFrontEndPreview
      True Vector=&Lerp_Vector(A=<1 0.9 1>, B=<1 1.5 1>, Alpha=$User.ParaGlideLeanAlpha)
      False Vector=&Lerp_Vector(A=<1 0.9 1>, B=<1 1.5 1>, Alpha=$User.ParaGlideLeanAlpha)
     )
    ))
   &SkeletalMeshLocation(
    Bone / Socket Index=&SelectIntFromArray(Direct Array Index=$Particles.PartitionParticles.Partition, Int Selection Array=DataInterface() {propertyValues {IntData [0,1]}})
    MaskSpawnGroup=false
    Mesh Sampling Type=::NewEnumerator5
    Orientation Sampling=::NewEnumerator1
    Sampled Position Offset=<0 0 25>
    Sampled Position Offset Coordinate Space:ENiagaraSkelMeshTransforms=::NewEnumerator2
    Skeletal Mesh=$System.Skeletal Mesh
    Socket Sampling Mode=::NewEnumerator5
   )
   &DynamicMaterialParameters(
    Index 0 Param 1=&RandomRangeFloat(Minimum=0.2, Maximum=1)
    Index 0 Param 2=&RandomRangeFloat(Minimum=0.2, Maximum=1)
    Index 0 Param 3=&FN_FrontEndFloatModifier(Gameplay=0.5, Front End=0.5)
    Index 0 Param 4=20
    Index 3 Param 1=300
    Index 3 Param 2=1
    Index 3 Param 4=&FloatFromCurve(FloatCurve=DataInterface() {propertyValues {Curve {keys [<at 0 0>,<at 1 1>]}}})
    Module.Index0Param0Write=true
    Module.Index0Param1Write=true
    Module.Index0Param2Write=true
    Module.Index0Param3Write=true
    Module.Index2Param0Write=true
    Module.Index2Param1Write=true
    Module.Index2Param2Write=true
    Module.Index2Param3Write=true
    Module.Index3Param0Write=true
    Module.Index3Param1Write=true
    Module.Index3Param2Write=true
    Module.Index3Param3Write=true
    Write Parameter Index 2=true
    Write Parameter Index 3=true
   )
   &InterpolateOverTime(
    Rate Of Change:NiagaraFloat=&FN_FrontEndFloatModifier(Gameplay=10, Front End=10)
    Target Vector Value:Vector3f=&AddVectorToPosition(
     Position=$Particles.Position
     Vector=&TransformVector(
      Vector=&FN_FrontEndVector3Modifier(Gameplay=$System.IGCenterposition, Front End=$System.FECenterposition)
      Destination Space=::Simulation
     )
    )
    Variable Type=::NewEnumerator2
   )
   /SetVariables(Particles.Interpolate Dynamic=&MakeFloat4FromVectorAndFloat(
     "Vector (RGB)"=$StackContext.InterpolateOverTime.InterpolateOverTime_V3_.Moving Average
     "Float (Alpha)"=&FN_FrontEndFloatModifier(Gameplay=1, Front End=0)
    )) {isSetParameters=true}
   /SetVariables(Particles.Color002=&Multiply_LinearColorByFloat(
     Linear Color=&MultiplyLinearColor(
      Linear Color=&MakeCustomLinearColorFromBool(Bool=$User.bIsFrontEndPreview, True Color=<rgba 0.0219 0.0346 0.0729 1>, False Color=<rgba 0.1 0.1 0.1 1>)
      Scale Factor=&MakeVector4_FromLinearColor(LinearColor=&FNTimeOfDayColor(Night Color=<rgba 0.1927 0.3945 1 1>, Morning Color=<rgba 0.6 0.6 0.6 1>, Day Color=<rgba 0.8 0.8 0.8 1>, Evening Color=<rgba 0.6 0.6 0.6 1>))
     )
     Float=&Multiply_Float(
      A=&FN_FrontEndFloatModifier(Gameplay=0.6, Front End=0.6)
      B=&FN_PickFloatDependOnPlatform(TargetPlatformActive=1, TargetPlatformNoActive=0.7)
     )
    )) {isSetParameters=true}
   /SetVariables(Particles.Size_Detail:Vector3f=&VectorFromCurve(VectorCurve=DataInterface() {propertyValues {
       XCurve {defaultValue=1, keys [{tangentMode=RCTM_User, time=0, value=0}, {tangentMode=RCTM_User, time=1, value=1}]}
       YCurve {defaultValue=1, keys [{tangentMode=RCTM_User, time=0, value=0}, {tangentMode=RCTM_User, time=1, value=1}]}
       ZCurve {defaultValue=1, keys [{tangentMode=RCTM_User, time=0, value=0}, {tangentMode=RCTM_User, time=1, value=1}]}
      }})) {isSetParameters=true}
   &ScaleColor(
    Scale Alpha=&FloatFromCurve(
     FloatCurve=DataInterface() {propertyValues {Curve {
        defaultValue=0
        keys [
         {arriveTangent=4.1785, interpMode=RCIM_Cubic, leaveTangent=4.1785, tangentMode=RCTM_User, time=0, value=0}
         {arriveTangent=1.1179, interpMode=RCIM_Cubic, leaveTangent=1.1179, tangentMode=RCTM_User, time=0.2042, value=0.6228}
         {arriveTangent=0.0224, interpMode=RCIM_Cubic, leaveTangent=0.0224, tangentMode=RCTM_User, time=1, value=1}
        ]
       }}}
     CurveIndex=$Emitter.Age
    )
    Scale RGB=&Multiply_Vector(
     A=&MakeVectorFromLinearColorRGB(Color=&FNTimeOfDayColor(Night Color=<rgba 0.5 0.625 1 1>, Morning Color=<rgba 1 1 1 1>, Day Color=<rgba 1 1 1 1>, Evening Color=<rgba 1 1 1 1>))
     B=&VectorFromFloat(Value=&FN_FrontEndFloatModifier(Gameplay=1, Front End=0.3))
    )
    ScaleA=true
    ScaleRGB=true
   )
   &UpdateMeshOrientation(
    Facing Interpolation=false
    Orientation Method=::NewEnumerator1
    Override Quat=false
    TransformFacing=false
    TransformUp=false
    Up Direction:Vector3f=&FN_FrontEndVector3Modifier(Gameplay=<0 0 1>, Front End=<0 -0.2 1>)
   ) {module=UpdateMeshOrientation001}
  ]
  rendererClasses [&NiagaraMeshRendererProperties]
  renderers [
   {rendererClass=&NiagaraMeshRendererProperties, rendererIndex=0}
   {rendererClass=&NiagaraMeshRendererProperties, rendererIndex=1}
  ]
 }
}
summary {
 systemName=NS_Contrail_Season41
 emitters [
  {emitterName=Base, rendererClasses [&NiagaraMeshRendererProperties]}
  {emitterName=RockShards_Tail, rendererClasses [&NiagaraMeshRendererProperties]}
  {emitterName=Glows, rendererClasses [&NiagaraSpriteRendererProperties, &NiagaraLightRendererProperties]}
  {emitterName=TrailLong, rendererClasses [&NiagaraMeshRendererProperties]}
  {emitterName=Lightning, rendererClasses [&NiagaraMeshRendererProperties]}
  {emitterName=Cubes_Trail, rendererClasses [&NiagaraMeshRendererProperties]}
  {emitterName=Cubes_Filler, rendererClasses [&NiagaraMeshRendererProperties]}
 ]
 userVariables [
  User.bIsFrontEnd {type=NiagaraBool}
  User.bIsFrontEndPreview {type=NiagaraBool}
  User.ParaGlideLeanAlpha {type=NiagaraFloat, defaultValue=0, description=""}
  User.Variant {type=NiagaraFloat, defaultValue=0, description=""}
 ]
}
